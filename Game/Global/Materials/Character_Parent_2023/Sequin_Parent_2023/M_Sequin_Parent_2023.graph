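# Parent character material that adds sequins to skinned, clothing and other mesh surfaces, using material attributes with several shading models.
# The shared PreFX character function builds the base material, which MF_Sequins modifies before the shared PostFX function finishes it.
# Sequin parameters supply tiled roughness, offset and normal textures, two falloff colors, rotation and a thin-film spectrum texture.
# A channel-mask parameter selects which channel of the PreFX Tech Art Mask feeds the sequin effect.
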Material /Game/Global/Materials/Character_Parent_2023/Sequin_Parent_2023/M_Sequin_Parent_2023
graph {
 properties {
  bUseMaterialAttributes=true
  bUsedWithClothing=true
  bUsedWithMeshDeformer=true
  bUsedWithMorphTargets=true
  bUsedWithNanite=true
  bUsedWithNiagaraMeshParticles=true
  bUsedWithSkeletalMesh=true
  pixelDepthOffsetMode=PDOM_Legacy
  shadingModel=MSM_FromMaterialExpression
  usedShadingModels=Default Lit | Subsurface | Clear Coat | Subsurface Profile | Cloth
 }
 nodes [
  n1 = MaterialFunctionCall() {materialFunction=&MF_CharacterParent_2023_PreFX}
  n2 = TextureObjectParameter() {
   group=Sequin
   parameterName=SequinRoughness
   samplerType=SAMPLERTYPE_LinearGrayscale
   sortPriority=0
   texture=/Game/Global/Textures/Default/Tileable_Patterns/T_SequinTile_roughness
  }
  n3 = TextureObjectParameter() {
   group=Sequin
   parameterName=SequinOffest
   samplerType=SAMPLERTYPE_LinearGrayscale
   sortPriority=1
   texture=/Game/Global/Textures/Default/Tileable_Patterns/T_SequinTile
  }
  n4 = TextureObjectParameter() {
   group=Sequin
   parameterName=SequinNormal
   samplerType=SAMPLERTYPE_Normal
   sortPriority=2
   texture=/Game/Global/Textures/Default/Tileable_Patterns/T_SequinTile_N
  }
  n5 = VectorParameter() {group=Sequin, parameterName=SequinFalloffColor01, sortPriority=3, defaultValue=<rgba 0.8542 0.387 0.5588 1>}
  n6 = VectorParameter() {group=Sequin, parameterName=SequinFalloffColor02, sortPriority=4, defaultValue=<rgba 0.3993 0.2651 0.8438 1>}
  MaterialFunctionCall(
   Material=MaterialFunctionCall!(
    inMat=@n1.Result
    SequinRoughness=@n2
    SequinOffset=@n3
    SequinNormal=@n4
    SequinFalloffColor01=@n5.RGB
    SequinFalloffColor02=@n6.RGB
    SequinRotationAngle=ScalarParameter() {group=Sequin, parameterName=SequinRotationAngle}
    SequinThinFilmTexture=TextureObjectParameter() {group=Sequin, parameterName=SequinThinFilm, texture=/Game/Global/Textures/Default/Thinfilm/T_ThinFilm_Spectrum_BlueGreen_Color}
    TechArtMask=ChannelMaskParameter(None=@n1.Tech Art Mask) {group=Sequin, parameterName=SequinChannel, sortPriority=0}.R
   ) {materialFunction=/Game/Global/MaterialFunctions/TechArt/MF_Sequins}.Result
   M=@n1.M
   Pre Skinned Local Position=@n1.M
   Distance Blend=@n1.M
   Fresnel=@n1.M
  ) {materialFunction=&MF_CharacterParent_2023_PostFX}
 ]
}
